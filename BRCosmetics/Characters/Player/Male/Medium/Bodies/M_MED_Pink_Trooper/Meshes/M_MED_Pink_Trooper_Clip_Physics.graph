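# Physics asset for the clip attachment on the M_MED_Pink_Trooper male player cosmetic body.
# A constraint chain links spine_01 through base_clip and dyn_ini_clip to a boxed dyn_clip, with capsules on the other bodies.
# The upper joints swing and twist freely, while the final clip joint locks twist and one swing axis, leaving a single swing axis.
# An unconstrained right-thigh capsule supplies an additional body beside the chain.
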PhysicsAsset /BRCosmetics/Characters/Player/Male/Medium/Bodies/M_MED_Pink_Trooper/Meshes/M_MED_Pink_Trooper_Clip_Physics
bodies {
 base_clip [{shapeType=Capsule, length=0.505, radius=0.505, shapeName=base_clip_capsule, rotation=<rot 0 -90 0>}]
 dyn_clip [{shapeType=Box, extentX=13.6363, extentY=7.3312, extentZ=1.5693, center=<6.0445 0.8768 -0.8491>, rotation=<rot -4.5318 8.067 8.5558>}]
 dyn_ini_clip [{
   shapeType=Capsule
   length=2.2571
   radius=1.3699
   shapeName=dyn_ini_clip_capsule
   center=<1.2441 0.0314 -0.0948>
   rotation=<rot 0 1.4751 -88.4402>
  }]
 spine_01 [{shapeType=Capsule, length=5.9788, radius=10.5212}]
 thigh_r [{shapeType=Capsule, length=22.6361, radius=8.7842, center=<17.8321 -0.6744 -0.0395>, rotation=<rot -88.8233 171.2937 -169.281>}]
}
constraints [
 {
  bone1Name=dyn_clip
  bone2Name=dyn_ini_clip
  swing1LimitDegrees=10
  swing1Motion=Locked
  swing2LimitDegrees=8
  twistLimitDegrees=5
  twistMotion=Locked
 }
 {
  bone1Name=base_clip
  bone2Name=spine_01
  swing1LimitDegrees=45
  swing1Motion=Free
  swing2LimitDegrees=45
  swing2Motion=Free
  twistLimitDegrees=45
  twistMotion=Free
 }
 {
  bone1Name=dyn_ini_clip
  bone2Name=base_clip
  swing1LimitDegrees=45
  swing1Motion=Free
  swing2LimitDegrees=45
  swing2Motion=Free
  twistLimitDegrees=45
  twistMotion=Free
 }
]
