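# Physics asset for the male medium Pink Trooper body mesh, covering the dynamic gear pouches rather than a full ragdoll.
# Box bodies cover the pouches, while box and capsule shapes cover the pelvis, lower spine chain and thighs.
# Tight angular limits tie three front pouches to spine_04 and the right side pouch to spine_01.
# The left pouch chains strap to parent to base, with no constraint from base to body.
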PhysicsAsset /BRCosmetics/Characters/Player/Male/Medium/Bodies/M_MED_Pink_Trooper/Meshes/M_MED_Pink_Trooper_Physics
bodies {
 dyn_pouch_base_l [{shapeType=Box, extentX=1.5, extentY=1.5, extentZ=1.5}]
 dyn_pouch_front_l [{shapeType=Box, extentX=13.0015, extentY=6.9854, extentZ=3.9043, center=<5.0258 0 0>}]
 dyn_pouch_front_mid [{shapeType=Box, extentX=12.5177, extentY=7.0922, extentZ=5.4694, center=<4.6582 0 0>}]
 dyn_pouch_front_r [{shapeType=Box, extentX=12.4406, extentY=6.6985, extentZ=4.9549, center=<4.9914 0 0>}]
 dyn_pouch_parent_l [{shapeType=Box, extentX=10.2357, extentY=8.273, extentZ=4.1947, center=<4.4793 -0.0002 -0.1259>}]
 dyn_pouch_side_r [{shapeType=Box, extentX=10.9526, extentY=10.1082, extentZ=5.1975, center=<4.5903 0.0002 2.3898>}]
 dyn_pouch_strap_l [{shapeType=Box, extentX=3.2087, extentY=2.6072, extentZ=1.5548, center=<1.5418 0 0>}]
 pelvis [{shapeType=Box, extentX=3, extentY=3, extentZ=3}]
 spine_01 [{shapeType=Capsule, length=7.5, radius=9}]
 spine_02 [{shapeType=Box, extentX=7.5, extentY=7.5, extentZ=7.5}]
 spine_03 [{shapeType=Box, extentX=7.5, extentY=7.5, extentZ=7.5}]
 spine_04 [{shapeType=Box, extentX=3.5, extentY=3.5, extentZ=3.5}]
 thigh_l [{shapeType=Capsule, length=26.5, radius=6.75, center=<-15.8225 0.0001 0.6448>, rotation=<rot 86.6613 -179.9999 -179.9999>}]
 thigh_r [
  {shapeType=Capsule, length=26.2891, radius=8.2109, center=<17.9891 -0.785 -0.0431>, rotation=<rot -88.8233 171.2917 -169.131>}
  {shapeType=Box, extentX=1, extentY=1, extentZ=1}
 ]
}
constraints [
 {bone1Name=dyn_pouch_side_r, bone2Name=spine_01, swing1LimitDegrees=10, swing2LimitDegrees=20, twistLimitDegrees=5}
 {bone1Name=dyn_pouch_parent_l, bone2Name=dyn_pouch_base_l, swing1LimitDegrees=8, swing2LimitDegrees=8, twistLimitDegrees=3}
 {bone1Name=dyn_pouch_strap_l, bone2Name=dyn_pouch_parent_l, swing1LimitDegrees=3, swing2LimitDegrees=15, twistLimitDegrees=2}
 {bone1Name=dyn_pouch_front_r, bone2Name=spine_04, swing1LimitDegrees=8, swing2LimitDegrees=6, twistLimitDegrees=3}
 {bone1Name=dyn_pouch_front_mid, bone2Name=spine_04, swing1LimitDegrees=8, swing2LimitDegrees=6, twistLimitDegrees=3}
 {bone1Name=dyn_pouch_front_l, bone2Name=spine_04, swing1LimitDegrees=8, swing2LimitDegrees=6, twistLimitDegrees=3}
]
