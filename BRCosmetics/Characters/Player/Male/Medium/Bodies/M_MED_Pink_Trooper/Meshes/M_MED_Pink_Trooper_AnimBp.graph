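# Secondary-motion AnimBP for the Pink Trooper cosmetic body, animating its clip and pouch bones over the pose copied from the attached parent mesh.
# base_clip is constrained by weights to pelvis, spine and right thigh twist, then offset.
# Two RigidBody passes and a spring-driven AnimDynamics clip follow, with additive bone corrections.
# EventGraph derives leg swing, speed and LOD2 values, plus per-movement-state gravity and sim-space settings through a deprecated helper.
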AnimBlueprint /BRCosmetics/Characters/Player/Male/Medium/Bodies/M_MED_Pink_Trooper/Meshes/M_MED_Pink_Trooper_AnimBp
target_skeleton=&/M_MED_Pink_Trooper_Skeleton
graphs [
 AnimGraph {nodes [
   n1 = "Constraint - base_clip (PPPPP)"!(
    ComponentPose=Local To Component(LocalPose=Copy Pose From Mesh() {bUseAttachedParent=true})
    Alpha=1
    ConstraintWeights_0=0.7214
    ConstraintWeights_1:float=0.2073
    ConstraintWeights_2:float=0.0657
    ConstraintWeights_3:float=0.0048
    ConstraintWeights_4:float=0.0008
   ) -> (Pose) {
    alphaBoolBlend {blendOption=Linear}
    boneToModify {boneName=base_clip}
    constraintSetup [
     {targetBone {boneName=pelvis}}
     {targetBone {boneName=spine_01}}
     {targetBone {boneName=thigh_twist_01_r}}
     {targetBone {boneName=spine_02}}
     {targetBone {boneName=spine_03}}
    ]
    constraintWeights [1,1,1,1,1]
   }
   n2 = "Transform (Modify) Bone - Bone: base_clip"!(ComponentPose=@n1, Alpha=1, Translation="0.935,0.000,1.911", Rotation="-37.279,2.656,-15.422") -> (Pose) {
    rotationMode=BMM_Additive
    rotationSpace=BCS_BoneSpace
    translationMode=BMM_Additive
    translationSpace=BCS_BoneSpace
    alphaBoolBlend {blendOption=Linear}
    boneToModify {boneName=base_clip}
    rotation=<rot -37.279 2.656 -15.422>
    translation {x=0.935, z=1.911}
   }
   n3 = "Transform (Modify) Bone - Bone: base_clip"(ComponentPose=@n2, Alpha=1, Translation="-4.494,-0.633,-0.281", Rotation="90,175.09,173.147") {
    rotationMode=BMM_Additive
    rotationSpace=BCS_BoneSpace
    translationMode=BMM_Additive
    translationSpace=BCS_BoneSpace
    alphaBoolBlend {blendOption=Linear}
    alphaScaleBiasClamp {interpSpeedDecreasing=0.1, interpSpeedIncreasing=0.1}
    boneToModify {boneName=base_clip}
    rotation=<rot 90 175.09 173.147>
    translation=<-4.494 -0.633 -0.281>
   }
   n4 = RigidBody!(
    ComponentPose=@n3
    Alpha
    OverrideWorldGravity:Vector="0,0,-980"
    ExternalForce
    SimSpaceSettings:Sim Space Settings Structure=(WorldAlpha=1,VelocityScaleZ=1,DampingAlpha=1,MaxLinearVelocity=3,MaxAngularVelocity=1000,MaxLinearAcceleration=1,MaxAngularAcceleration=1,ExternalLinearDragV=(X=1,Y=1,Z=1),ExternalLinearVelocity=(X=0,Y=0,Z=0),ExternalAngularVelocity=(X=0,Y=0,Z=0))
   ) -> (Pose) {
    bClampLinearTranslationLimitToRefPose=true
    bOverrideWorldGravity=true
    overridePhysicsAsset=&/M_MED_Pink_Trooper_Physics
    simulationSpace=BaseBoneSpace
    alphaBoolBlend {blendOption=Linear}
    baseBoneRef {boneName=pelvis}
    overrideWorldGravity {z=-980}
    simSpaceSettings {maxAngularAcceleration=1, maxAngularVelocity=1000, maxLinearAcceleration=1, maxLinearVelocity=3, worldAlpha=1, externalLinearDragV=<1 1 1>}
   }
   n5 = RigidBody(
    ComponentPose=@n4
    OverrideWorldGravity="0,0,-980"
    SimSpaceSettings=(WorldAlpha=1,VelocityScaleZ=1,DampingAlpha=1,MaxLinearVelocity=3,MaxAngularVelocity=1000,MaxLinearAcceleration=1,MaxAngularAcceleration=1,ExternalLinearDragV=(X=1,Y=1,Z=1),ExternalLinearVelocity=(X=0,Y=0,Z=0),ExternalAngularVelocity=(X=0,Y=0,Z=0))
   ) {
    bClampLinearTranslationLimitToRefPose=true
    bOverrideWorldGravity=true
    overridePhysicsAsset=&/M_MED_Pink_Trooper_Clip_Physics
    simulationSpace=BaseBoneSpace
    alphaBoolBlend {blendOption=Linear}
    baseBoneRef {boneName=thigh_r}
    overrideWorldGravity {z=-980}
    simSpaceSettings {maxAngularAcceleration=1, maxAngularVelocity=1000, maxLinearAcceleration=1, maxLinearVelocity=3, worldAlpha=1, externalLinearDragV=<1 1 1>}
   }
   n6 = "Transform (Modify) Bone - Bone: dyn_pouch_base_l"(ComponentPose=@n5, Alpha=1, Translation="0,-4,0", Rotation="0,0,-20", Scale="1,1,1") {
    rotationMode=BMM_Additive
    rotationSpace=BCS_BoneSpace
    translationMode=BMM_Additive
    translationSpace=BCS_BoneSpace
    alphaBoolBlend {blendOption=Linear}
    boneToModify {boneName=dyn_pouch_base_l}
    rotation {roll=-20}
    translation {y=-4}
   }
   n7 = Anim Dynamics - Bone: dyn_clip(ComponentPose=@n6, Alpha=1, ExternalForce="0,0,0") {
    angularDampingOverride=1
    angularSpringConstant=20000
    bAngularSpring=true
    bOverrideAngularDamping=true
    bUseGravityOverride=true
    gravityScale=0.35
    simulationSpace=RootRelative
    alphaBoolBlend {blendOption=Linear}
    boundBone {boneName=dyn_clip}
    componentLinearAccScale=<1 1 1>
    componentLinearVelScale=<1 1 1>
    gravityOverride {z=-980}
    physicsBodyDefinitions [{
      boundBone {boneName=dyn_clip}
      boxExtents=<45 45 45>
      constraintSetup {angularLimitsMax=<0 0 0>, angularLimitsMin=<0 -80 0>, angularTarget=<1 0 0>}
      localJointOffset=<8 0 0>
     }]
   }
   n8 = "Transform (Modify) Bone - Bone: dyn_ini_clip"(ComponentPose=@n7, Alpha=1, Translation="-4.397,-1.723,2.010", Rotation="8.754,-1.454,-0.377", Scale="1,1,1") {
    rotationMode=BMM_Additive
    rotationSpace=BCS_BoneSpace
    translationMode=BMM_Additive
    translationSpace=BCS_BoneSpace
    alphaScaleBiasClamp {bInterpResult=true}
    boneToModify {boneName=dyn_ini_clip}
    rotation=<rot 8.754 -1.454 -0.377>
    translation=<-4.397 -1.723 2.01>
   }
   Output Pose(Result=Component To Local(ComponentPose=@n8)) {layerGroup=None}
  ]}
 EventGraph {nodes [
   Event BlueprintUpdateAnimation()
   > n1 = Sequence!(execute) -> (then_0, then_1, then_2:Exec, then_3:Exec)
   n2 = Get bIsSprinting()
   n3 = Branch(execute=@n1.then_1, Condition=AND Boolean(A=Get bIsCrouching(), B=NOT Boolean(A=@n2)))
   > Set NewVar(NewVar:double=-120) -> (Output_Get:double)
   Branch(execute=@n3.else, Condition=OR Boolean(A=Get bIsCrouchSprinting(), B=Get bIsPlayingEmote()))
   > Set NewVar(NewVar=-100)
   n4 = Gravity Override State Function DEPRECATED(
    execute=@n1.then_3
    CharacterPartAnimInstance=Self-Reference()
    GameTimeInSeconds=GetGameTimeInSeconds()
    GravityOverrideParams=Get Gravity Override Params()
    RBAN Params=Get RBAN Params()
   )
   > Set GravityOverride(GravityOverride=@n4)
   > Set Sim Space Settings(Sim Space Settings:Sim Space Settings Structure=Break RBANSetting Struct!(RBANSettingStruct:"RBANSetting Struct Structure (by ref)"=@n4.RBAN SettingsOut) -> (
     ComponentLinearAccScale_28_#1:Vector=~
     ComponentLinearVelScale_31_#2:Vector=~
     SimSpaceSettings_34_#3:Sim Space Settings Structure
    ) {is=title K2Node_BreakStruct}.SimSpaceSettings_34_#3) -> (Output_Get:Sim Space Settings Structure)
   n5 = Get PawnSpeed2D()
   Set PlayerSpeed(execute=@n1.then_2, PlayerSpeed:double=SelectFloat(A=0.5, B=1, bPickA="float >= float"(A=@n5, B=325) {is=title K2Node_CallFunction})) -> (Output_Get:double)
   n6 = BreakTransform(InTransform=GetDeltaTransformFromRefPose(self=GetOwnerSkeletalMeshComponent(), BoneName=thigh_r, BaseName=pelvis))
   Set legSwingDelta_r(
    execute=@n1.then_0
    legSwingDelta_r:double=MapRangeClamped(Value=BreakRotator(InRot=@n6.Rotation).Yaw, InRangeA=0, InRangeB=Get NewVar() -> (NewVar:double))
   ) -> (Output_Get:double)
   > n7 = Set LOD2__Bool(LOD2__Bool:bool="Equal (Integer)"(A=Get CurrentLOD(), B=2)) -> (Output_Get:bool)
   > Set LOD2(LOD2:double="To Float (Boolean)"(InBool=@n7)) -> (Output_Get:double)
  ]}
]
variables [
 Gravity Override Params {type=GravityOverrideParamsStruct, default {tactical_Sprint {gravityOverride=<0 0 -300>}}}
 GravityOverride {type=Vector}
 legSwingDelta_r {type=number}
 LOD2 {type=number}
 LOD2__Bool {type=boolean}
 NewVar {type=number, default=-120}
 PlayerSpeed {type=number}
 RBAN Params {
  type=RBANParamsStruct
  default {
   dBNO {simSpaceSettings {
     maxAngularAcceleration=1
     maxAngularVelocity=1000
     maxLinearAcceleration=1
     maxLinearVelocity=3
     velocityScaleZ=1
     externalLinearDragV=<1 1 1>
    }}
   dBNO_Carried {simSpaceSettings {
     maxAngularAcceleration=1
     maxAngularVelocity=1000
     maxLinearAcceleration=1
     maxLinearVelocity=3
     velocityScaleZ=1
     externalLinearDragV=<1 1 1>
    }}
   emoteOrMelee {simSpaceSettings {
     maxAngularAcceleration=1
     maxAngularVelocity=1000
     maxLinearAcceleration=1
     maxLinearVelocity=3
     velocityScaleZ=1
     externalLinearDragV=<1 1 1>
    }}
   front_End {simSpaceSettings {
     maxAngularAcceleration=1
     maxAngularVelocity=1000
     maxLinearAcceleration=1
     maxLinearVelocity=3
     velocityScaleZ=0
     worldAlpha=0.5
     externalLinearDragV=<0 0 0>
    }}
   onGround_Crouch_Moving {simSpaceSettings {
     dampingAlpha=0.1
     maxAngularAcceleration=1
     maxAngularVelocity=1000
     maxLinearAcceleration=1
     maxLinearVelocity=3
     velocityScaleZ=0.1
     worldAlpha=0.1
     externalLinearDragV=<1 1 1>
    }}
   onGround_Falling {simSpaceSettings {
     maxAngularAcceleration=1
     maxAngularVelocity=1000
     maxLinearAcceleration=1
     maxLinearVelocity=3
     velocityScaleZ=1
     externalLinearDragV=<1 1 1>
    }}
   onGround_Moving {simSpaceSettings {
     dampingAlpha=0.2
     maxAngularAcceleration=1
     maxAngularVelocity=1000
     maxLinearAcceleration=1
     maxLinearVelocity=3
     velocityScaleZ=1
     worldAlpha=0.2
     externalLinearDragV=<1 1 1>
    }}
   onGround_Standing {simSpaceSettings {
     maxAngularAcceleration=1
     maxAngularVelocity=1000
     maxLinearAcceleration=1
     maxLinearVelocity=3
     velocityScaleZ=1
     externalLinearDragV=<1 1 1>
    }}
   rail_Grinding {simSpaceSettings {
     maxAngularAcceleration=1
     maxAngularVelocity=1000
     maxLinearAcceleration=1
     maxLinearVelocity=3
     velocityScaleZ=1
     worldAlpha=1
     externalLinearDragV=<1 1 1>
    }}
   riding_Moving {simSpaceSettings {
     maxAngularAcceleration=1
     maxAngularVelocity=1000
     maxLinearAcceleration=1
     maxLinearVelocity=3
     velocityScaleZ=1
     externalLinearDragV=<1 1 1>
    }}
   riding_Moving_Targeting {simSpaceSettings {
     maxAngularAcceleration=1
     maxAngularVelocity=1000
     maxLinearAcceleration=1
     maxLinearVelocity=3
     velocityScaleZ=1
     externalLinearDragV=<1 1 1>
    }}
   riding_Standing {simSpaceSettings {
     maxAngularAcceleration=1
     maxAngularVelocity=1000
     maxLinearAcceleration=1
     maxLinearVelocity=3
     velocityScaleZ=1
     externalLinearDragV=<1 1 1>
    }}
   skydive_DivingDown {simSpaceSettings {
     maxAngularAcceleration=1
     maxAngularVelocity=1000
     maxLinearAcceleration=1
     maxLinearVelocity=3
     velocityScaleZ=1
     externalLinearDragV=<1 1 1>
    }}
   skydive_FloatingDown {
    componentLinearAccScale=<0 0.5 0>
    componentLinearVelScale=<0 0.5 0>
    simSpaceSettings {
     maxAngularAcceleration=1
     maxAngularVelocity=1000
     maxLinearAcceleration=1
     maxLinearVelocity=3
     velocityScaleZ=1
     externalLinearDragV=<1 1 1>
    }
   }
   skydive_GoingUp {simSpaceSettings {
     maxAngularAcceleration=1
     maxAngularVelocity=1000
     maxLinearAcceleration=1
     maxLinearVelocity=3
     velocityScaleZ=1
     externalLinearDragV=<1 1 1>
    }}
   skydive_Parachuting {
    componentLinearAccScale=<0 0.1 0.3>
    componentLinearVelScale=<0 0.1 0.3>
    simSpaceSettings {
     maxAngularAcceleration=1
     maxAngularVelocity=1000
     maxLinearAcceleration=1
     maxLinearVelocity=3
     velocityScaleZ=1
     externalLinearDragV=<1 1 1>
    }
   }
   skydive_Surfing {simSpaceSettings {
     maxAngularAcceleration=1
     maxAngularVelocity=1000
     maxLinearAcceleration=1
     maxLinearVelocity=3
     velocityScaleZ=1
     externalLinearDragV=<1 1 1>
    }}
   swimming {simSpaceSettings {
     maxAngularAcceleration=1
     maxAngularVelocity=1000
     maxLinearAcceleration=1
     maxLinearVelocity=3
     velocityScaleZ=1
     externalLinearDragV=<1 1 1>
    }}
   tactical_Sprint {simSpaceSettings {
     maxAngularAcceleration=1
     maxAngularVelocity=1000
     maxLinearAcceleration=1
     maxLinearVelocity=3
     velocityScaleZ=1
     externalLinearDragV=<1 1 1>
    }}
   ziplining {simSpaceSettings {
     maxAngularAcceleration=1
     maxAngularVelocity=1000
     maxLinearAcceleration=1
     maxLinearVelocity=3
     velocityScaleZ=1
     externalLinearDragV=<1 1 1>
    }}
  }
 }
 RBAN Settings Out {
  type=RBANSettingStruct
  default {
   componentLinearAccScale=<0 0 0>
   componentLinearVelScale=<0 0 0>
   simSpaceSettings {
    dampingAlpha=1
    maxAngularAcceleration=1000
    maxAngularVelocity=3
    maxLinearAcceleration=300
    maxLinearVelocity=700
    velocityScaleZ=0
    worldAlpha=1
    externalAngularVelocity=<0 0 0>
    externalLinearDragV=<0 0.5 0>
    externalLinearVelocity=<0 0 0>
   }
  }
 }
 Sim Space Settings {
  type=SimSpaceSettings
  default {
   dampingAlpha=1
   maxAngularAcceleration=10000
   maxAngularVelocity=10000
   maxLinearAcceleration=10000
   maxLinearVelocity=10000
   velocityScaleZ=1
   worldAlpha=0
   externalLinearDragV=<0 0 0>
  }
 }
]
