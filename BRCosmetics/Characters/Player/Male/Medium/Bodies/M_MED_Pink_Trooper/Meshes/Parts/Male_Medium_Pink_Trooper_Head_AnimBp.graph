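# Animation Blueprint for the head part of the male medium Pink Trooper cosmetic body, correcting head rotation on top of the body's animation.
# Each update, Scale Joint Rotation Function derives a scaled rotation from head and neck_02, stored in head rot.
# When crouching and targeting, its yaw is damped further.
# The AnimGraph copies the attached parent's pose and replaces head_offset's bone-space rotation with head rot.
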AnimBlueprint /BRCosmetics/Characters/Player/Male/Medium/Bodies/M_MED_Pink_Trooper/Meshes/Parts/Male_Medium_Pink_Trooper_Head_AnimBp
target_skeleton=&/Male_Medium_Pink_Trooper_Head_Skeleton
graphs [
 AnimGraph {nodes [
   n1 = "Transform (Modify) Bone - Bone: head_offset"!(
    ComponentPose=Local To Component(LocalPose=Copy Pose From Mesh() {bUseAttachedParent=true})
    Alpha=1
    Rotation=Get head rot() -> (head rot:Rotator)
   ) -> (Pose) {
    rotationMode=BMM_Replace
    rotationSpace=BCS_BoneSpace
    alphaBoolBlend {blendOption=Linear}
    alphaScaleBiasClamp {interpSpeedDecreasing=5, interpSpeedIncreasing=20}
    boneToModify {boneName=head_offset}
   }
   Output Pose(Result=Component To Local(ComponentPose=@n1)) {layerGroup=None}
  ]}
 EventGraph {nodes [
   Event BlueprintUpdateAnimation()
   > n1 = Scale Joint Rotation Function!(
    execute
    Scale:Rotator="1,0.55,0.6"
    TargetBone:Name=head
    ReferenceBone:Name=neck_02
    SkelMeshComponent:Skinned Mesh Component Object Reference=GetPartSkeletalMeshComponent()
   ) -> (then, FinalRotation:Rotator, OriginalRotation:Rotator)
   > Set head rot(head rot:Rotator=@n1) -> (Output_Get:Rotator)
   > Branch(Condition=AND Boolean(A=Get bIsCrouching(), B=Get bIsTargeting()))
   n2 = BreakRotator(InRot=Get head rot())
   > Set head rot(head rot=MakeRotator(Roll=@n2, Pitch=@n2.Pitch, Yaw=float * float(A=@n2.Yaw, B=0.4) {is=title K2Node_PromotableOperator}))
   TryGetPawnOwner()
  ]}
]
variables [head rot {type=Rotator}]
